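# Material for the Dusty Diner sign lights, with three mask-channel light groups that animate over time or follow a parameter.
# The Timeline_Base static switch selects Light_Controls RGB or a time-driven mask sample, dotted with the emissive mask to set lit areas.
# The same mask blends yellow, orange and brown tints for DeriveHDRfromLDR.
# Its output multiplies saturation-adjusted diffuse glow and an inverted Fresnel, alongside normal and specular-suffixed mask textures.
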Material /Game/Environments/Props/Signs/Materials/M_Dusty_Diner_Lights
graph {
 properties {bUsedWithNanite=true, pixelDepthOffsetMode=PDOM_Legacy, usedShadingModels=Default Lit}
 nodes [
  n1 = VectorParameter() {desc="Use the RGB channels to control the 3 lights being on and off.  ", parameterName=Light_Controls, defaultValue=<rgba 0 0 0 1>}
  n2 = Multiply(A=ScalarParameter() {defaultValue=0.275, parameterName=Red}, B=Time())
  n3 = TextureSample() {samplerType=SAMPLERTYPE_Masks, texture=/Game/Environments/Props/Signs/Textures/T_Dusty_Diner_Sign_Emissives_M}
  n4 = DotProduct(
   A=StaticSwitchParameter(
    True=@n1.RGB
    False=TextureSample(UVs=@n2) {samplerType=SAMPLERTYPE_Masks, texture=/Game/Environments/Props/Signs/Textures/T_Dusty_Diner_Sign_Emissives_M2}.RGB
   ) {desc="To turn on control of the lights via RGB color variable.  ", parameterName=Timeline_Base}
   B=@n3.RGB
  )
  n5 = VectorParameter() {parameterName=Yellow, defaultValue=<rgba 0.5 0.5 0 1>}
  n6 = LinearInterpolate(
   A=LinearInterpolate(A=Multiply(A=@n3.R, B=@n5.RGB), B=VectorParameter() {parameterName=Orange, defaultValue=<rgba 0.6615 0.2262 0 1>}.RGB, Alpha=@n3.G)
   B=VectorParameter() {parameterName=Brown, defaultValue=<rgba 1 0.2172 0 0>}.RGB
   Alpha=@n3.B
  )
  n7 = MaterialFunctionCall(
   LDR Input=@n4
   HDR Dynamic Range=ScalarParameter() {defaultValue=100, parameterName=HDR_Dynamic_Range}
   HDR Tint=@n6
   HDR Intensity Multiplier=ScalarParameter() {defaultValue=0.5, parameterName=HDR_Intensity}
   Derive HDR Power=ScalarParameter() {defaultValue=0.5, parameterName=Derive}
  ) {materialFunction=&DeriveHDRfromLDR}
  n8 = Desaturation(
   None=TextureSample() {texture=/Game/Environments/Props/Signs/Textures/T_Dusty_Diner_Lights_D}.RGB
   Fraction=ScalarParameter() {defaultValue=-1.5, parameterName=Saturation}
  )
  Multiply(A=@n7.RemappedRange, B=Multiply(A=Multiply(A=@n4, B=@n8), B=ScalarParameter() {defaultValue=10, parameterName=EmissivePower}))
  Multiply(A=@n6, B=@n4)
  n9 = ScalarParameter() {defaultValue=3, parameterName=Fresnel}
  Multiply(A=OneMinus(None=Fresnel(ExponentIn=@n9) {baseReflectFraction=0}), B=@n7.RemappedRange)
  TextureSample() {samplerType=SAMPLERTYPE_Normal, texture=/Game/Environments/Props/Signs/Textures/T_Dusty_Diner_Lights_N}
  TextureSample() {samplerType=SAMPLERTYPE_Masks, texture=/Game/Environments/Props/Signs/Textures/T_Dusty_Diner_Lights_S}
 ]
}
